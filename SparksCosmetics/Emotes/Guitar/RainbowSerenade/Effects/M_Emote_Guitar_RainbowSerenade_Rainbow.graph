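# Unlit alpha-composite material for Niagara mesh particles that draws the rainbow in the Rainbow Serenade guitar emote.
# Mirrored, quantized U bands feed a rainbow color function, while a dynamic Erode parameter reveals them along mirrored V with glowing, faded tips.
# World-aligned screen-space noise erodes in sparkles at the reveal edge, and axis-aligned fresnel and depth fade soften the result.
# Timeline, debug and dither nodes sit unconnected.
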Material /SparksCosmetics/Emotes/Guitar/RainbowSerenade/Effects/M_Emote_Guitar_RainbowSerenade_Rainbow
graph {
 properties {
  bAutomaticallySetUsageInEditor=false
  bUsedWithNiagaraMeshParticles=true
  blendMode=BLEND_AlphaComposite
  shadingModel=MSM_Unlit
  usedShadingModels=Unlit
 }
 nodes [
  n1 = MaterialFunctionCall() {materialFunction=/Game/Effects/Fort_Effects/MaterialFunctions/MF_U}
  n2 = Constant() {r=0.5}
  n3 = ScalarParameter() {defaultValue=1.666, desc="Bunches the colors a bit,\r\nadjust according to fresnel", parameterName="Rainbow SCurve ", sliderMax=3}
  n4 = MaterialFunctionCall!(
   In=MaterialFunctionCall!(Gradient=Saturate(None=@n1.U), Center Point=@n2) {materialFunction=/Game/Effects/Fort_Effects/MaterialFunctions/MF_GradientReflected}.Result
   Power=@n3
  ) {materialFunction=/Game/Effects/Fort_Effects/MaterialFunctions/MF_SCurve_Float}
  n5 = ScalarParameter() {defaultValue=6, parameterName=Steps}
  n6 = Subtract(
   A=OneMinus(None=MaterialFunctionCall!(Gradient=@n4.Result, Steps=@n5) {materialFunction=/Game/Effects/Fort_Effects/MaterialFunctions/MF_Quantize}.Default Range)
   B=Divide(B=@n5)
  )
  n7 = ComponentMask(None=MaterialFunctionCall!(Gradient=Frac(None=@n6)) {materialFunction=/BRCosmetics/Athena/Effects/Particles/Skydiving/NS_Contrail_DarkAndLight/MF_DarkAndLight_RainbowColors}.Result) {b=true, g=true, r=true}
  n8 = VectorParameter() {parameterName=Rainbow Reveal Color, defaultValue=<rgba 3 1.3305 1.11 2>}
  n9 = ScalarParameter() {defaultValue=0.1, parameterName=Tip Glow Size}
  n10 = MaterialFunctionCall(Gradient=MaterialFunctionCall() {materialFunction=/Game/Effects/Fort_Effects/MaterialFunctions/MF_V}.V, Center Point=@n2) {materialFunction=&MF_GradientReflected}
  n11 = Saturate(None=Divide(A=Subtract(A=@n10.Result, B=@n9), B=@n9))
  n12 = MaterialFunctionCall!(
   Gradient=@n10.Result
   Erosion Amount=OneMinus(None=DynamicParameter() {paramNames [Erode,Param2,Param3,Param4]}.Erode)
   Softness=ScalarParameter() {defaultValue=0.75, parameterName=Sparkle Gradient Size, sliderMax=1}
  ) {materialFunction=/BRCosmetics/Athena/Effects/Particles/Emotes/NS_Emote_Spectrum/Functions/MF_Erosion}
  n13 = Power(Base=OneMinus(None=Multiply(A=@n11, B=@n12.Increased Range)), Exp=@n8.A)
  n14 = Divide(
   A=Subtract(A=@n10.Result, B=ScalarParameter() {defaultValue=0.075, parameterName=Bottom Fade Shift})
   B=ScalarParameter() {defaultValue=0.222, parameterName=Bottom Fade Size, sliderMax=1}
  ) {constB=0.1}
  n15 = Multiply(A=Round(None=@n12.Increased Range) {desc=Rainbow Opacity as it reveals}, B=Power(Base=Saturate(None=@n14), Exp=@n8.A)) {desc=Multiplied by fade on tip}
  n16 = AppendVector(A=Multiply(A=LinearInterpolate(A=@n7, B=@n8.RGB, Alpha=@n13), B=@n15), B=@n15)
  n17 = MaterialFunctionCall(Axis=CrossProduct(A=VertexNormalWS(), B=VertexTangentWS())) {materialFunction=&AxisAlignedFresnel}
  n18 = MaterialFunctionCall(In=@n17.Fresnel, Power=ScalarParameter() {defaultValue=2, parameterName=Fresnel Exponent Rainbow, sliderMax=4}) {materialFunction=&MF_SCurve_Float}
  n19 = VectorParameter() {parameterName=Sparkle Color, defaultValue=<rgba 0 33 31.35 1>}
  n20 = ObjectPositionWS()
  n21 = TextureSample(UVs=Divide(
    A=MaterialFunctionCall!(Center Point World Position=@n20) {materialFunction=/Engine/Functions/Engine_MaterialFunctions02/UVs/WorldSpaceAlignedScreenCoordinates}.Offset UVs
    B=ScalarParameter() {defaultValue=20, parameterName=Tiling Size}
   )) {samplerType=SAMPLERTYPE_LinearGrayscale, texture=/BRCosmetics/Athena/Effects/Particles/Emotes/NS_Emote_Spectrum/T_Emote_Spectrum_Noise}
  n22 = Multiply(A=@n21.R, B=SquareRoot(None=@n17.Fresnel)) {desc=Multiplying fresnel}
  n23 = MaterialFunctionCall(Gradient=@n12.Increased Range, Center Point=Constant() {r=0.5}) {materialFunction=&MF_GradientReflected}
  n24 = SquareRoot(None=OneMinus(None=Multiply(A=@n23.Result, B=@n11))) {desc="This gets rid of\r\nthe harsh cut at\r\nthe tips"}
  n25 = DepthFade(
   Opacity=LinearInterpolate(
    A=Multiply(A=@n16, B=@n18.Result) {desc=Multiplying fresnel}
    B=AppendVector(A=@n19.RGB, B=@n19.RGB)
    Alpha=MaterialFunctionCall!(Gradient=@n22, Erosion Amount=@n24) {materialFunction=/Game/Effects/Fort_Effects/MaterialFunctions/MF_ErosionNormalized}.Result
   )
   FadeDistance=ScalarParameter() {parameterName=Depth Fade Distance, sliderMax=33}
  )
  ComponentMask(None=@n25) {a=true}
  Divide(A=@n9)
  n26 = Divide(
   A=Subtract(A=Time(), B=ScalarParameter() {bUseCustomPrimitiveData=true, parameterName=ActivationTime, sliderMax=1})
   B=ScalarParameter() {bUseCustomPrimitiveData=true, defaultValue=3.1, parameterName=ActivationDuration, primitiveDataIndex=1}
  )
  MaterialFunctionCall!(Number To Convert=OneMinus(None=Saturate(None=@n26)), Number of Bits, UVs) {materialFunction=/Engine/Functions/Engine_MaterialFunctions01/Debug/DebugBinaryValues-Float}
  MaterialFunctionCall() {materialFunction=/Game/Effects/Fort_Effects/MaterialFunctions/MF_Emote_Spectrum_Timeline}
  n27 = ComponentMask(None=TextureCoordinate()) {r=true}
  n28 = OneMinus(None=@n27)
  n29 = Add(A=Multiply(A=Multiply(A=@n28, B=@n27)) {constB=4}) {constB=0.2}
  n30 = ParticleColor()
  n31 = ComponentMask(None=@n25) {b=true, g=true, r=true}
  MaterialFunctionCall(Emissive=Multiply(A=Saturate(None=@n29), B=Multiply(A=@n30.RGB, B=@n31))) {materialFunction=&MF_FX_EmissiveAdjustments}
  n32 = Subtract(A=MaterialFunctionCall() {materialFunction=&DitherTemporalAA}.Result) {constB=0.5}
  n33 = ScalarParameter() {parameterName=Dither Intensity}
  VertexInterpolator(VS=Add(A=Multiply(A=@n32, B=@n33), B=@n1.U) {desc="Dither, skip for now"}) {desc=prevents cleanup}
 ]
}
