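# Niagara one-shot mesh cloud poof for the Rainbow Serenade guitar emote.
# A single burst emitter spawns randomly rotated meshes from a small sphere, with user variables controlling count, size and velocity.
# Linear drag slows the short-lived puffs while a curve pops their scale up, then shrinks it to zero.
# Toon shader parameters and a curve-animated dynamic material parameter shape the look, and time-of-day inputs adjust the color.
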NiagaraSystem /SparksCosmetics/Emotes/Guitar/RainbowSerenade/Effects/NS_Emote_Guitar_RainbowSerenade_CloudPoof
dependencies {usedDataInterfaces [&NiagaraDataInterfaceCurve], usedDynamicInputs [], usedModules [&SystemState]}
emitters {CloudPoof {
  emitterSpawnScript [/Game/Effects/Niagara/MasterMaterials/ToonShader/Template/NM_ToonShaderParameters(HighLight_Softness=0, Lighting_Hardness=0.8, Metallic_LightingPower=0, RimLight_Color:LinearColor=<rgba 0 0 0 1>, RimLight_Width=0)]
  emitterUpdateScript [
   &EmitterState()
   &SpawnBurst_Instantaneous(Spawn Count=$User.SpawnCount, Use Spawn Probability=false, UseLoopCountLimit=false)
  ]
  particleSpawnScript [
   &InitializeParticle(
    Color=&HueShiftLinearColor(
     Hue=0.7
     Saturation=&FNTimeOfDayScalar(Night Scale=0.7, Morning Scale=0.7, Day Scale=0.7, Evening Scale=0.5)
     Value=&RandomRangeFloat(Minimum=2, Maximum=1.5)
     Linear Color=<rgba 1 1 1 1>
    )
    Lifetime=&RandomRangeFloat(Minimum=0.3, Maximum=0.6)
    Mesh Scale Mode=::NewEnumerator3
    Mesh Uniform Scale=&Multiply_Float(A=&RandomRangeFloat(Minimum=1, Maximum=2), B=$User.Scale)
    Sprite Size Mode=::NewEnumerator3
    Uniform Sprite Size=3.1789
    UsePositionOffset=false
    Write Material Random=true
   )
   &InitialMeshOrientation(Rotation=&RandomRangeVector(Minimum=<0 0 0>, Maximum=<1 1 1>), Use Rotation Vector=true)
   &ShapeLocation(MaskSpawnGroup=false, Override Seed=false, Sphere Radius=1, UseOwnerScale=true) {module=ShapeLocation001}
   &AddVelocity(
    Module.Use Falloff Distance=false
    Override Seed=false
    Velocity Mode=::NewEnumerator1
    Velocity Speed=&RandomRangeFloat(Minimum=300, Maximum=400)
   )
   /SetVariables(Particles.Velocity=&Multiply_Vector(A=$Particles.Velocity, B=$User.VelocityScale)) {isSetParameters=true}
  ]
  particleUpdateScript [
   &ParticleState()
   &ScaleMeshSize(Scale Factor=&VectorFromFloat(Value=&FloatFromCurve(FloatCurve=DataInterface() {propertyValues {Curve {keys [
          {arriveTangent=5.3301, interpMode=RCIM_Cubic, leaveTangent=5.3301, tangentMode=RCTM_User, time=0.1, value=0}
          {arriveTangent=-0.7334, interpMode=RCIM_Cubic, leaveTangent=-0.7334, tangentMode=RCTM_User, time=0.2, value=0.95}
          {arriveTangent=-1.0526, tangentMode=RCTM_User, time=1, value=0}
         ]}}}))) {module=ScaleMeshSize001}
   &Drag(Drag=&RandomRangeFloat(Minimum=4, Maximum=6), Use Linear Drag=true, Use Rotational Drag=false)
   &SolveForcesAndVelocity(Clamp Velocity=false, Limit Acceleration=false)
   &DynamicMaterialParameters(
    Index 0 Param 1=&FloatFromCurve(FloatCurve=DataInterface() {propertyValues {Curve {keys [{interpMode=RCIM_Cubic, tangentMode=RCTM_User, time=0, value=0}, {interpMode=RCIM_Cubic, tangentMode=RCTM_User, time=1, value=0.2}]}}})
    Module.Param0WriteEnabled=true
    Module.Param1WriteEnabled=true
    Module.Param2WriteEnabled=true
    Module.Param3WriteEnabled=true
   )
   &ScaleColor(
    Scale RGB=&MakeVectorFromLinearColorRGB(Color=&FNTimeOfDayColor(
      Night Color=<rgba 0.2941 0.3646 0.7 1>
      Morning Color=<rgba 0.625 0.625 0.625 1>
      Day Color=<rgba 0.6667 0.6667 0.6667 1>
      Evening Color=<rgba 0.6667 0.6667 0.6667 1>
     ))
    ScaleA=true
    ScaleRGB=true
   )
  ]
  rendererClasses [&NiagaraMeshRendererProperties]
  renderers [{rendererClass=&NiagaraMeshRendererProperties, rendererIndex=0}]
 }}
summary {
 systemName=NS_Emote_Guitar_RainbowSerenade_CloudPoof
 emitters [{emitterName=CloudPoof, rendererClasses [&NiagaraMeshRendererProperties]}]
 userVariables [
  User.bIsFrontEndPreview {type=NiagaraBool}
  User.Scale {type=NiagaraFloat, defaultValue=1, description=""}
  User.SpawnCount {type=NiagaraInt32, defaultValue=15}
  User.VelocityScale {type=Vector3f, description="", defaultValue {struct=Vector3f, value=<1 1 1>}}
 ]
}
