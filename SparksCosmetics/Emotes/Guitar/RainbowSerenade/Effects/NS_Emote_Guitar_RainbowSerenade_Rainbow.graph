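# Niagara effect for the Rainbow Serenade guitar emote: a rainbow mesh with side clouds, trailing cloud puffs and multicolored sparkle sprites, all placed by skeletal mesh bone sampling.
# Rainbow and side-cloud meshes burst-spawn and keep sampling their bones, with a system-age curve driving a rainbow material parameter.
# Traversal puffs and sparkles spawn per distance moved.
# Front-end modifiers add spawn rates and adjust velocity, size and offsets.
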NiagaraSystem /SparksCosmetics/Emotes/Guitar/RainbowSerenade/Effects/NS_Emote_Guitar_RainbowSerenade_Rainbow
dependencies {
 usedDataInterfaces [
  &NiagaraDataInterfaceCurve
  &NiagaraDataInterfaceArrayColor
  &NiagaraDataInterfaceArrayFloat2
  &NiagaraDataInterfaceArrayFloat
  &NiagaraDataInterfaceArrayFloat3
  &NiagaraDataInterfaceVectorCurve
  &NiagaraDataInterfaceColorCurve
  &NiagaraDataInterfaceSkeletalMesh
 ]
 usedDynamicInputs []
 usedModules [
  &self:SystemSpawnScript.SystemScriptSource.SystemScriptGraph.NiagaraNodeAssignment_0.SetVariables_#1
  &SystemState
 ]
}
emitters {
 Rainbow {
  emitterUpdateScript [
   &EmitterState()
   &SpawnBurst_Instantaneous(Use Spawn Probability=false, UseLoopCountLimit=true)
  ]
  particleSpawnScript [
   &InitializeParticle(Mesh Scale=<1.275 3 1.275>, Mesh Scale Mode=::NewEnumerator4, UsePositionOffset=false, Write Material Random=true)
   &SkeletalMeshLocation(
    Bone / Socket Index=3
    Bone Sampling Mode=::NewEnumerator4
    MaskSpawnGroup=false
    Sampled Position Offset=<0 -30 -35>
    Skeletal Mesh=$System.Skeletal Mesh
   ) {module=SkeletalMeshLocation001}
  ]
  particleUpdateScript [
   &ParticleState(Kill Particles When Lifetime Has Elapsed=false, Let Infinitely Lived Particles Die When Emitter Deactivates=true)
   &DynamicMaterialParameters(
    Index 0 Param 1=&FloatFromCurve(FloatCurve=DataInterface() {propertyValues {Curve {keys [<at 0 0>, {leaveTangent=2.6501, time=1.4, value=1}]}}}, CurveIndex=$System.Age)
    Module.Index0Float0Write=true
    Module.Index0Float1Write=true
    Module.Index0Float2Write=true
    Module.Index0Float3Write=true
   )
   &SkeletalMeshLocation(
    Bone / Socket Index=3
    Bone Sampling Mode=::NewEnumerator4
    MaskSpawnGroup=false
    Sampled Position Offset=<-5 10 -34>
    Skeletal Mesh=$System.Skeletal Mesh
   )
   &InterpolateOverTime(
    Initial Position Value:NiagaraPosition=$Particles.Initial.Position
    Rate Of Change:NiagaraFloat=&RemapRange(
     Input Value To Remap=&VectorLength(Vector=$Particles.SkeletalMeshLocation.SampledVelocity)
     Input Min=0
     Input Max=500
     Output Min=0
     Output Max=150
    )
    Target Position Value:NiagaraPosition=$Particles.SkeletalMeshLocation.SampledPosition
    Variable Type=::NewEnumerator6
   )
  ]
  rendererClasses [&NiagaraMeshRendererProperties]
  renderers [{rendererClass=&NiagaraMeshRendererProperties, rendererIndex=0}]
 }
 SeatCloudsTraversal {
  emitterSpawnScript [/Game/Effects/Niagara/MasterMaterials/ToonShader/Template/NM_ToonShaderParameters(HighLight_Softness=0, Lighting_Hardness=0.8, Metallic_LightingPower=0, RimLight_Color:LinearColor=<rgba 0 0 0 1>, RimLight_Width=0)]
  emitterUpdateScript [
   &EmitterState()
   &SpawnRate(SpawnRate=&FN_FrontEndFloatModifier(Gameplay=0, Front End=20), Use Spawn Probability=false)
   &SpawnPerUnit(
    Module.Use Max Movement Threshold=true
    Module.Use Movement Tolerance=true
    Movement Tolerance=1
    Spawn Spacing=3
    Use Spawn Probability=false
   )
  ]
  particleSpawnScript [
   &InitializeParticle(
    Lifetime=&RandomRangeFloat(Minimum=0.5, Maximum=0.6)
    Mesh Scale Mode=::NewEnumerator1
    Mesh Uniform Scale Max=0.6
    Mesh Uniform Scale Min=0.5
    UsePositionOffset=true
    Write Material Random=true
   )
   &InitialMeshOrientation(Rotation=&RandomRangeVector(Minimum=<0 0 0>, Maximum=<1 1 1>), Use Rotation Vector=true)
   &CameraOffset(Camera Offset Amount=&RandomRangeFloat(Minimum=0, Maximum=10)) {enabled=false}
   &AddVelocity(
    Velocity=&RandomRangeVector(Minimum=<0 -35 -3>, Maximum=<0 -35 -3>)
    Velocity Speed Scale=&FN_FrontEndFloatModifier(Gameplay=1, Front End=5)
   ) {module=AddVelocity002}
   /SetVariables(Particles.Velocity=&Multiply_Vector(A=$Particles.Velocity, B=<1 1 1>)) {enabled=false, isSetParameters=true, module=SetVariables_#2}
   &ShapeLocation(
    Handle Radius=15
    Large Radius=15
    MaskSpawnGroup=false
    Non Uniform Scale=<1 1 0.6>
    Offset Mode=::NewEnumerator0
    Override Seed=false
    Shape Primitive=::NewEnumerator3
    U Distribution=0.5
    UseOwnerScale=true
   ) {enabled=false, module=ShapeLocation002}
   /SetVariables(Particles.VisibilityTag=&RandomRangeInt(Minimum=0, Maximum=1)) {enabled=false, isSetParameters=true, module=SetVariables_#3}
   &SkeletalMeshLocation(
    Bone / Socket Index=4
    Bone Sampling Mode=::NewEnumerator4
    MaskSpawnGroup=false
    Sampled Position Offset=<0 -5 -12>
    Skeletal Mesh=$System.Skeletal Mesh
   )
   &ShapeLocation(MaskSpawnGroup=false, Non Uniform Scale=<5 1 2>, Offset Mode=::NewEnumerator0, Override Seed=false, Sphere Radius=3, UseOwnerScale=true) {enabled=false}
   &ShapeLocation(
    Box Size=<55 10 5>
    MaskSpawnGroup=false
    Offset=<0 0 -2>
    Offset Mode=::NewEnumerator0
    Override Seed=false
    Shape Primitive=::NewEnumerator2
    Surface Only=false
    UseOwnerScale=true
   ) {module=ShapeLocation001}
  ]
  particleUpdateScript [
   &ParticleState()
   &UpdateMeshOrientation(Rotation Rate=0.4, Rotation Vector=&RandomRangeVector(Minimum=<-1 -1 -1>, Maximum=<1 1 1>))
   &ScaleMeshSize(Scale Factor=&VectorFromFloat(Value=&FloatFromCurve(FloatCurve=DataInterface() {propertyValues {Curve {keys [
          {arriveTangent=7.3539, time=0, value=0}
          {arriveTangent=-0.0014, interpMode=RCIM_Cubic, leaveTangent=-0.0014, tangentMode=RCTM_User, time=0.1, value=1}
          {arriveTangent=-1.0747, interpMode=RCIM_Cubic, leaveTangent=-1.0747, tangentMode=RCTM_User, time=1, value=0}
         ]}}}))) {module=ScaleMeshSize001}
   &Drag(Drag=2, Use Linear Drag=true, Use Rotational Drag=false)
   &PointAttractionForce(
    Attraction Radius=1000
    Attraction Strength=500
    Attractor Position Offset:Vector3f=<0 -30 0>
    Kill Within Radius=false
    Module.Use Falloff=true
    UseAttractorOffset=true
   ) {enabled=false}
   &SolveForcesAndVelocity(Clamp Velocity=false, Limit Acceleration=false)
   &DynamicMaterialParameters(
    Index 0 Param 1=&FloatFromCurve(FloatCurve=DataInterface() {propertyValues {Curve {keys [{interpMode=RCIM_Cubic, tangentMode=RCTM_User, time=0, value=0}, {interpMode=RCIM_Cubic, tangentMode=RCTM_User, time=1, value=0.2}]}}})
    Module.Param0WriteEnabled=true
    Module.Param1WriteEnabled=true
    Module.Param2WriteEnabled=true
    Module.Param3WriteEnabled=true
   )
   &ScaleColor(
    Scale Alpha=&FloatFromCurve(FloatCurve=DataInterface() {propertyValues {Curve {keys [
         {arriveTangent=0.0022, interpMode=RCIM_Cubic, leaveTangent=0.0022, tangentMode=RCTM_User, time=0, value=1}
         {arriveTangent=0.1254, interpMode=RCIM_Cubic, tangentMode=RCTM_Break, time=1, value=0.4}
        ]}}})
    Scale RGB=&VectorFromFloat(Value=&FNTimeOfDayScalar(Night Scale=0.35, Morning Scale=1, Day Scale=0.6, Evening Scale=1.4))
    ScaleA=true
    ScaleRGB=true
   ) {enabled=false, module=ScaleColor001}
   &ScaleColor(
    Scale RGB=&MakeVectorFromLinearColorRGB(Color=&FNTimeOfDayColor(
      Night Color=<rgba 0.2941 0.3646 0.7 1>
      Morning Color=<rgba 0.625 0.625 0.625 1>
      Day Color=<rgba 0.6667 0.6667 0.6667 1>
      Evening Color=<rgba 0.6667 0.6667 0.6667 1>
     ))
    ScaleA=true
    ScaleRGB=true
   ) {enabled=false}
   &ScaleColor(
    Scale RGB=&MakeVectorFromLinearColorRGB(Color=<rgba 0.6302 0.6302 0.6302 1>)
    ScaleA=true
    ScaleRGB=true
   ) {module=ScaleColor003}
  ]
  rendererClasses [&NiagaraMeshRendererProperties]
  renderers [{rendererClass=&NiagaraMeshRendererProperties, rendererIndex=0}]
 }
 SideClouds {
  emitterSpawnScript [&NM_ToonShaderParameters(HighLight_Softness=0, Lighting_Hardness=3.5, Metallic_LightingPower=0, RimLight_Color:LinearColor=<rgba 0 0 0 1>, RimLight_Width=0)]
  emitterUpdateScript [
   &EmitterState()
   &SpawnBurst_Instantaneous(Spawn Time=0.35, Use Spawn Probability=false, UseLoopCountLimit=true)
   &SpawnBurst_Instantaneous(Spawn Group=1, Spawn Time=0.5, Use Spawn Probability=false, UseLoopCountLimit=true) {module=SpawnBurst_Instantaneous001}
  ]
  particleSpawnScript [
   &PartitionParticles()
   &InitializeParticle(Lifetime=3, Mesh Scale Mode=::NewEnumerator3, Mesh Uniform Scale=0.97, Position Mode=::NewEnumerator0, Write Material Random=true)
   &InitialMeshOrientation(
    Mesh Orientation Mode=::NewEnumerator2
    Orientation Vector=&SelectVectorFromArray(
     Direct Array Index=&SpawnGroup()
     Vector Selection Array=DataInterface() {propertyValues {FloatData [<0 1 1>, <0 -1 -1>]}}
    )
    Use Rotation Vector=false
   )
   &SkeletalMeshLocation(
    Bone / Socket Index=4
    Bone Sampling Mode=::NewEnumerator4
    MaskSpawnGroup=false
    Sampled Position Offset=&SelectVectorFromArray(
     Direct Array Index=&SpawnGroup()
     Vector Selection Array=DataInterface() {propertyValues {FloatData [<60 -30 0>, <-60 -30 0>]}}
    )
    Skeletal Mesh=$System.Skeletal Mesh
   )
  ]
  particleUpdateScript [
   &ParticleState(Kill Particles When Lifetime Has Elapsed=false, Let Infinitely Lived Particles Die When Emitter Deactivates=true)
   &ScaleMeshSize(Scale Factor=<1 0.9 1>) {module=ScaleMeshSize001}
   &ScaleMeshSize(Scale Factor=&VectorFromFloat(Value=&FloatFromCurve(FloatCurve=DataInterface() {propertyValues {Curve {keys [
          <at 0 0 RCIM_Cubic>
          {arriveTangent=6.0085, interpMode=RCIM_Cubic, leaveTangent=6.0085, time=0.0648, value=1.1904}
          {arriveTangent=-0.2222, leaveTangent=-0.2222, time=0.1648, value=0.9904}
          <at 1 1>
         ]}}})))
   &SkeletalMeshLocation(
    Bone / Socket Index=4
    Bone Sampling Mode=::NewEnumerator4
    MaskSpawnGroup=false
    Orientation Sampling=::NewEnumerator1
    Sampled Position Offset=&FN_FrontEndVector3Modifier(
     Gameplay=&SelectVectorFromArray(
      Direct Array Index=&SpawnGroup()
      Vector Selection Array=DataInterface() {propertyValues {FloatData [<54 -5 0>, <-56.5 -5 0>]}}
     )
     Front End=&SelectVectorFromArray(
      Direct Array Index=&SpawnGroup()
      Vector Selection Array=DataInterface() {propertyValues {FloatData [<53 -5 0>, <-56.5 -5 0>]}}
     )
    )
    Skeletal Mesh=$System.Skeletal Mesh
   ) {module=SkeletalMeshLocation002}
   &UpdateMeshOrientation(Facing Interpolation=false, Orientation Method=::NewEnumerator1, Override Quat=false, TransformFacing=false, TransformUp=false)
   &InterpolateOverTime(
    Initial Position Value:NiagaraPosition=$Particles.Initial.Position
    Rate Of Change:NiagaraFloat=&RemapRange(
     Input Value To Remap=&VectorLength(Vector=$Particles.SkeletalMeshLocation002.SampledVelocity)
     Input Min=0
     Input Max=500
     Output Min=&FloatFromCurve(FloatCurve=DataInterface() {propertyValues {Curve {keys [<at 0 1>,<at 0.4 1>,<at 0.5 0>]}}}, CurveIndex=$System.Age, Scale Curve=150)
     Output Max=150
    )
    Target Position Value:NiagaraPosition=$Particles.SkeletalMeshLocation002.SampledPosition
    Variable Type=::NewEnumerator6
   ) {enabled=false}
   /SetVariables(Particles.Position=$Particles.InterpolateOverTime.InterpolateOverTime_Position.Moving Average) {enabled=false, isSetParameters=true}
   &InheritVelocity(Inherited Velocity Speed Limit=1, Speed Limit=true, UseSpeedThreshold=false) {enabled=false}
   &SolveForcesAndVelocity(Clamp Velocity=false, Limit Acceleration=false) {enabled=false}
   &ScaleColor(
    Scale RGB=&MakeVectorFromLinearColorRGB(Color=&FNTimeOfDayColor(
      Night Color=<rgba 0.526 0.526 0.526 1>
      Morning Color=<rgba 0.526 0.526 0.526 1>
      Day Color=<rgba 0.526 0.526 0.526 1>
      Evening Color=<rgba 0.526 0.526 0.526 1>
     ))
    ScaleA=true
    ScaleRGB=true
   ) {enabled=false}
   &ScaleColor(
    Scale Mode=::NewEnumerator1
    Scale RGBA=&MakeVector4_FromLinearColor(LinearColor=$NPC.FortniteNPC.SunLightColor)
    ScaleRGBA=true
   ) {enabled=false, module=ScaleColor001}
   &ScaleColor(
    Scale RGB=&MakeVectorFromLinearColorRGB(Color=&ColorFromCurve(
      ColorCurve=DataInterface() {propertyValues {
        AlphaCurve {keys [<at 0 1 RCIM_Cubic>,<at 1 0 RCIM_Cubic>]}
        BlueCurve {keys [<at 0 0 RCIM_Cubic>,<at 1 1 RCIM_Cubic>]}
        GreenCurve {keys [<at 0 0 RCIM_Cubic>,<at 1 0 RCIM_Cubic>]}
        RedCurve {keys [<at 0 1 RCIM_Cubic>,<at 1 0.0643 RCIM_Cubic>]}
       }}
      CurveIndex=Linked() {linkedVariable {name=NPC.FortniteNPC.TimeOfDay_0-1, type=NiagaraFloat}}
     ))
    ScaleA=true
    ScaleRGB=true
   ) {enabled=false, module=ScaleColor002}
   &ScaleColor(
    Scale RGB=&MakeVectorFromLinearColorRGB(Color=<rgba 0.6302 0.6302 0.6302 1>)
    ScaleA=true
    ScaleRGB=true
   ) {module=ScaleColor003}
  ]
  rendererClasses [&NiagaraMeshRendererProperties]
  renderers [{rendererClass=&NiagaraMeshRendererProperties, rendererIndex=0}]
 }
 SideCloudsTraversal {
  emitterSpawnScript [&NM_ToonShaderParameters(HighLight_Softness=0, Lighting_Hardness=0.8, Metallic_LightingPower=0, RimLight_Color:LinearColor=<rgba 0 0 0 1>, RimLight_Width=0)]
  emitterUpdateScript [
   &EmitterState()
   &SpawnRate(SpawnRate=&FN_FrontEndFloatModifier(Gameplay=0, Front End=20), Use Spawn Probability=false)
   &SpawnPerUnit(
    Module.Use Max Movement Threshold=true
    Module.Use Movement Tolerance=true
    Movement Tolerance=1
    Spawn Spacing=5
    Use Spawn Probability=false
   )
  ]
  particleSpawnScript [
   &InitializeParticle(
    Lifetime=&RandomRangeFloat(Minimum=0.3, Maximum=0.4)
    Mesh Scale Mode=::NewEnumerator1
    Mesh Uniform Scale Max=0.5
    Mesh Uniform Scale Min=0.4
    UsePositionOffset=true
    Write Material Random=true
   )
   &InitialMeshOrientation(Rotation=&RandomRangeVector(Minimum=<0 0 0>, Maximum=<1 1 1>), Use Rotation Vector=true)
   &CameraOffset(Camera Offset Amount=&RandomRangeFloat(Minimum=0, Maximum=10)) {enabled=false}
   /SetVariables(Particles.Velocity=&Multiply_Vector(A=$Particles.Velocity, B=<1 1 1>)) {enabled=false, isSetParameters=true, module=SetVariables_#2}
   &ShapeLocation(
    Handle Radius=15
    Large Radius=15
    MaskSpawnGroup=false
    Non Uniform Scale=<1 1 0.6>
    Offset Mode=::NewEnumerator0
    Override Seed=false
    Shape Primitive=::NewEnumerator3
    U Distribution=0.5
    UseOwnerScale=true
   ) {enabled=false, module=ShapeLocation002}
   /SetVariables(Particles.VisibilityTag=&RandomRangeInt(Minimum=0, Maximum=1)) {enabled=false, isSetParameters=true, module=SetVariables_#3}
   &SkeletalMeshLocation(
    Bone / Socket Index=4
    Bone Sampling Mode=::NewEnumerator4
    MaskSpawnGroup=false
    Sampled Position Offset=&SelectVectorFromArray(
     Array Sampling Mode=::NewEnumerator0
     Direct Array Index=?
     Vector Selection Array=DataInterface() {propertyValues {FloatData [<55 -5 0>, <-55 -5 0>]}}
     Random Seed=0
    )
    Skeletal Mesh=$System.Skeletal Mesh
   )
   &ShapeLocation(MaskSpawnGroup=false, Offset Mode=::NewEnumerator0, Override Seed=false, Sphere Radius=5, UseOwnerScale=true)
   &AddVelocity(
    Velocity=&RandomRangeVector(Minimum=<0 -30 -3>, Maximum=<0 -30 -3>)
    Velocity Speed Scale=&FN_FrontEndFloatModifier(Gameplay=1, Front End=4)
   ) {module=AddVelocity002}
  ]
  particleUpdateScript [
   &ParticleState()
   &UpdateMeshOrientation(Rotation Rate=0.3, Rotation Vector=&RandomRangeVector(Minimum=<-1 -1 -1>, Maximum=<1 1 1>))
   &ScaleMeshSize(Scale Factor=&VectorFromFloat(Value=&FloatFromCurve(FloatCurve=DataInterface() {propertyValues {Curve {keys [
          {arriveTangent=7.3539, time=0, value=0}
          {arriveTangent=-0.0014, interpMode=RCIM_Cubic, leaveTangent=-0.0014, tangentMode=RCTM_User, time=0.1, value=1}
          {arriveTangent=-1.0747, interpMode=RCIM_Cubic, leaveTangent=-1.0747, tangentMode=RCTM_User, time=1, value=0}
         ]}}}))) {module=ScaleMeshSize001}
   &Drag(Drag=2, Use Linear Drag=true, Use Rotational Drag=false)
   &PointAttractionForce(
    Attraction Radius=1000
    Attraction Strength=500
    Attractor Position Offset:Vector3f=<0 -30 0>
    Kill Within Radius=false
    Module.Use Falloff=true
    UseAttractorOffset=true
   ) {enabled=false}
   &SolveForcesAndVelocity(Clamp Velocity=false, Limit Acceleration=false)
   &DynamicMaterialParameters(
    Index 0 Param 1=&FloatFromCurve(FloatCurve=DataInterface() {propertyValues {Curve {keys [{interpMode=RCIM_Cubic, tangentMode=RCTM_User, time=0, value=0}, {interpMode=RCIM_Cubic, tangentMode=RCTM_User, time=1, value=0.2}]}}})
    Module.Param0WriteEnabled=true
    Module.Param1WriteEnabled=true
    Module.Param2WriteEnabled=true
    Module.Param3WriteEnabled=true
   )
   &ScaleColor(
    Scale Alpha=&FloatFromCurve(FloatCurve=DataInterface() {propertyValues {Curve {keys [
         {arriveTangent=0.0022, interpMode=RCIM_Cubic, leaveTangent=0.0022, tangentMode=RCTM_User, time=0, value=1}
         {arriveTangent=0.1254, interpMode=RCIM_Cubic, tangentMode=RCTM_Break, time=1, value=0.4}
        ]}}})
    Scale RGB=&VectorFromFloat(Value=&FNTimeOfDayScalar(Night Scale=0.35, Morning Scale=1, Day Scale=0.6, Evening Scale=1.4))
    ScaleA=true
    ScaleRGB=true
   ) {enabled=false, module=ScaleColor001}
   &ScaleColor(
    Scale RGB=&MakeVectorFromLinearColorRGB(Color=&FNTimeOfDayColor(
      Night Color=<rgba 0.2941 0.3646 0.7 1>
      Morning Color=<rgba 0.625 0.625 0.625 1>
      Day Color=<rgba 0.6667 0.6667 0.6667 1>
      Evening Color=<rgba 0.6667 0.6667 0.6667 1>
     ))
    ScaleA=true
    ScaleRGB=true
   ) {enabled=false}
   &ScaleColor(
    Scale RGB=&MakeVectorFromLinearColorRGB(Color=<rgba 0.6302 0.6302 0.6302 1>)
    ScaleA=true
    ScaleRGB=true
   ) {module=ScaleColor003}
  ]
  rendererClasses [&NiagaraMeshRendererProperties]
  renderers [{rendererClass=&NiagaraMeshRendererProperties, rendererIndex=0}]
 }
 SparklesRainbow {
  emitterUpdateScript [
   &EmitterState()
   &SpawnPerUnit(Module.Use Max Movement Threshold=true, Module.Use Movement Tolerance=false, Spawn Spacing=15, Use Spawn Probability=false)
   &SpawnRate(SpawnRate=4, Use Spawn Probability=true)
   &SpawnRate(
    SpawnGroup=1
    SpawnRate=&FN_FrontEndFloatModifier(Gameplay=0, Front End=15)
    Use Spawn Probability=false
   ) {module=SpawnRate001}
  ]
  particleSpawnScript [
   &PartitionParticles()
   &InitializeParticle(
    Color=&SelectLinearColorFromArray(
     Array Sampling Mode=::NewEnumerator0
     Direct Array Index=?
     Color Selection Array=DataInterface() {propertyValues {ColorData [
        <rgba 13 0.7082 2.3551 1>
        <rgba 13 8.7277 0.2098 1>
        <rgba 13 12.7174 0 1>
        <rgba 0.1739 13 0.164 1>
        <rgba 0 11.6709 13 1>
        <rgba 1.8334 1.6442 13 1>
       ]}}
     Random Seed=0
    )
    Lifetime Max=0.666
    Lifetime Min=0.333
    Lifetime Mode=::NewEnumerator1
    Position Mode=::NewEnumerator0
    Sprite Rotation Angle=0.125
    Sprite Rotation Mode=::NewEnumerator1
    Sprite Size Mode=::NewEnumerator3
    Uniform Sprite Size=&Multiply_Float(
     A=&RandomRangeFloat(Minimum=20, Maximum=30)
     B=&FN_FrontEndFloatModifier(Gameplay=2, Front End=1.75)
    )
    Write Material Random=true
   )
   &AddVelocityInCone(
    Cone Angle=90
    Module.Use Velocity Falloff On Cone Axis=false
    Velocity Distribution Along Cone Axis=0.9
    Velocity Strength=&RandomRangeFloat(Minimum=11, Maximum=22)
   ) {enabled=false}
   &SphereLocation(
    MaskSpawnGroup=false
    Override Seed=false
    Sphere Radius=4
    Surface Only=false
    UseNonUniformScale=false
    UseOffset=false
    UseSphereOrientation=false
   ) {enabled=false}
   &ShapeLocation(
    Handle Radius=11
    Large Radius=63
    MaskSpawnGroup=false
    Non Uniform Scale=<0.75 1 0.5>
    Offset Mode=::NewEnumerator0
    Override Seed=false
    Rotation Axis=<0 1 0>
    Rotation Coordinate Space=::World
    Rotation Mode=::NewEnumerator1
    Shape Primitive=::NewEnumerator3
    U Distribution=0.5
    UseOwnerScale=true
   ) {module=ShapeLocation001}
   &SkeletalMeshLocation(
    Bone / Socket Index=4
    Bone Sampling Mode=::NewEnumerator4
    MaskSpawnGroup=false
    Position Sampling=::NewEnumerator2
    Sampled Position Offset=<0 -12 -14>
    Sampled Position Offset Coordinate Space:ENiagaraSkelMeshTransforms=::NewEnumerator2
    Skeletal Mesh=$System.Skeletal Mesh
   )
   &ShapeLocation(
    Box / Plane Mode:ENiagara_BoxPlaneMode=::NewEnumerator1
    MaskSpawnGroup=false
    Offset=&SelectVectorFromArray(
     Direct Array Index=$Particles.PartitionParticles.Partition
     Vector Selection Array=DataInterface() {propertyValues {FloatData [<0 0 70>, <-50 0 50>, <50 0 50>]}}
    )
    Offset Mode=::NewEnumerator0
    Override Seed=false
    Plane Size:Vector2f=&SelectVector2DFromArray(
     Direct Array Index=$Particles.PartitionParticles.Partition
     Vector 2D Selection Array=DataInterface() {propertyValues {FloatData [<60 35>, <60 30>, <60 30>]}}
    )
    Rotation Angle:NiagaraFloat=&SelectFloatFromArray(Direct Array Index=$Particles.PartitionParticles.Partition, Float Selection Array=DataInterface() {propertyValues {FloatData [0,0,0]}})
    Rotation Axis=<0 1 0>
    Rotation Mode=::NewEnumerator1
    Shape Primitive=::NewEnumerator2
    UseOwnerScale=true
   ) {enabled=false}
   /SetVariables(Particles.Position=&AddVectorToPosition(
     Position=$Particles.SkeletalMeshLocation.SampledPosition
     Vector=&RotateVectorByQuaternion(Vector To Rotate=$Particles.ShapeLocation001.ShapeVector, Quaternion=$Engine.Owner.Rotation)
    )) {isSetParameters=true}
   &AddVelocity(Velocity=&MaskVectorBySpawnGroup(Masked Vector=<0 0 0>, Passthrough Vector=&RandomRangeVector(Minimum=<0 -25 0>, Maximum=<0 -50 0>)))
  ]
  particleUpdateScript [
   &AccelerationForce(Acceleration=/Niagara/DynamicInputs/ScaleByCurve/ScaleVectorByCurve(
     InitialValue:Vector3f=<0 0 11>
     ScaleCurve:NiagaraDataInterfaceVectorCurve=DataInterface() {propertyValues {XCurve {keys []}, YCurve {keys []}, ZCurve {keys [<at 0 0>,<at 2 1>]}}}
     CurveIndex:NiagaraFloat=$Emitter.Age
    ))
   &CurlNoiseForce(Bias Noise=false, Pan Noise=false)
   &SolveForcesAndVelocity(Clamp Velocity=false, Limit Acceleration=false)
   &ParticleState()
   &ScaleSpriteSize(Scale Factor=&Vector2DFromFloat(Value=&FloatFromCurve(FloatCurve=DataInterface() {propertyValues {Curve {keys [<at 0 0>,<at 0.3 1>,<at 0.6 1.075>,<at 1 0>]}}})))
   &DynamicMaterialParameters(
    Index 0 Param 1=&FloatFromCurve(
     FloatCurve=DataInterface() {propertyValues {Curve {keys [{tangentMode=RCTM_User, time=0, value=0}, {tangentMode=RCTM_User, time=1, value=1}]}}}
     CurveIndex=$Engine.System.Age
    )
    Module.Index0Float0Write=true
    Module.Index0Float1Write=true
    Module.Index0Float2Write=true
    Module.Index0Float3Write=true
   )
   &CameraOffset(Camera Offset Amount=5) {module=CameraOffset001}
  ]
  rendererClasses [&NiagaraSpriteRendererProperties]
  renderers [
   {rendererClass=&NiagaraSpriteRendererProperties}
   {rendererClass=&NiagaraSpriteRendererProperties, rendererIndex=1}
  ]
 }
 SparklesTraversal {
  emitterUpdateScript [
   &EmitterState()
   &SpawnPerUnit(Module.Use Max Movement Threshold=true, Module.Use Movement Tolerance=false, Spawn Spacing=5, Use Spawn Probability=false)
   &SpawnRate(SpawnRate=&FN_FrontEndFloatModifier(Gameplay=0, Front End=15), Use Spawn Probability=true)
  ]
  particleSpawnScript [
   &InitializeParticle(
    Color=&SelectLinearColorFromArray(
     Array Sampling Mode=::NewEnumerator0
     Direct Array Index=?
     Color Selection Array=DataInterface() {propertyValues {ColorData [
        <rgba 13 0 0.3576 1>
        <rgba 13 8.7277 0.2098 1>
        <rgba 13 12.7174 0 1>
        <rgba 0.1739 13 0.164 1>
        <rgba 0 11.6709 13 1>
        <rgba 1.8333 1.6442 13 1>
       ]}}
     Random Seed=0
    )
    Lifetime Max=0.8
    Lifetime Min=0.6
    Lifetime Mode=::NewEnumerator1
    Sprite Rotation Angle=0.125
    Sprite Rotation Mode=::NewEnumerator1
    Sprite Size Mode=::NewEnumerator3
    Uniform Sprite Size=&Multiply_Float(
     A=&RandomRangeFloat(Minimum=20, Maximum=30)
     B=&FN_FrontEndFloatModifier(Gameplay=2, Front End=1)
    )
    UsePositionOffset=false
    Write Material Random=true
   )
   &PartitionParticles(Number Of Partitions=3) {enabled=false}
   &AddVelocityInCone(
    Cone Angle=90
    Module.Use Velocity Falloff On Cone Axis=false
    Velocity Distribution Along Cone Axis=0.9
    Velocity Strength=&RandomRangeFloat(Minimum=11, Maximum=22)
   ) {enabled=false}
   &SphereLocation(
    MaskSpawnGroup=false
    Override Seed=false
    Sphere Radius=4
    Surface Only=false
    UseNonUniformScale=false
    UseOffset=false
    UseSphereOrientation=false
   ) {enabled=false}
   &SkeletalMeshLocation(
    Bone / Socket Index=4
    Bone Sampling Mode=::NewEnumerator4
    MaskSpawnGroup=false
    Sampled Position Offset=<0 -30 5>
    Skeletal Mesh=$System.Skeletal Mesh
   )
   &ShapeLocation(MaskSpawnGroup=false, Override Seed=false, Sphere Radius=25, Sphere Surface Distribution=1, UseOwnerScale=true) {module=ShapeLocation002}
   &AddVelocity(Velocity=&RandomRangeVector(Minimum=<0 0 0>, Maximum=<0 -30 0>))
   &ShapeLocation(
    Handle Radius=13
    Large Radius=63
    MaskSpawnGroup=false
    Non Uniform Scale=<0.85 1 0.3>
    Offset Mode=::NewEnumerator0
    Override Seed=false
    Rotation Axis=<0 1 0>
    Rotation Mode=::NewEnumerator1
    Shape Primitive=::NewEnumerator3
    U Distribution=0.5
    UseOwnerScale=true
   ) {enabled=false, module=ShapeLocation001}
   &ShapeLocation(
    Box / Plane Mode:ENiagara_BoxPlaneMode=::NewEnumerator1
    MaskSpawnGroup=false
    Offset=&SelectVectorFromArray(
     Direct Array Index=$Particles.PartitionParticles.Partition
     Vector Selection Array=DataInterface() {propertyValues {FloatData [<0 0 70>, <-50 0 50>, <50 0 50>]}}
    )
    Offset Mode=::NewEnumerator0
    Override Seed=false
    Plane Size:Vector2f=&SelectVector2DFromArray(
     Direct Array Index=$Particles.PartitionParticles.Partition
     Vector 2D Selection Array=DataInterface() {propertyValues {FloatData [<60 35>, <60 30>, <60 30>]}}
    )
    Rotation Angle:NiagaraFloat=&SelectFloatFromArray(Direct Array Index=$Particles.PartitionParticles.Partition, Float Selection Array=DataInterface() {propertyValues {FloatData [0,0,0]}})
    Rotation Axis=<0 1 0>
    Rotation Mode=::NewEnumerator1
    Shape Primitive=::NewEnumerator2
    UseOwnerScale=true
   ) {enabled=false}
   &CameraOffset(Camera Offset Amount=50) {module=CameraOffset002}
   /SetVariables(Particles.DarkeningCardColor:LinearColor=<rgba 0.0312 0.0312 0.0312 1>) {isSetParameters=true}
  ]
  particleUpdateScript [
   &AccelerationForce(Acceleration=&ScaleVectorByCurve(
     InitialValue:Vector3f=<0 0 3>
     ScaleCurve:NiagaraDataInterfaceVectorCurve=DataInterface() {propertyValues {XCurve {keys []}, YCurve {keys []}, ZCurve {keys [<at 0 0>,<at 2 1>]}}}
     CurveIndex:NiagaraFloat=$Emitter.Age
    ))
   &SolveForcesAndVelocity(Clamp Velocity=false, Limit Acceleration=false)
   &ParticleState()
   &ScaleSpriteSize(Scale Factor=&Vector2DFromFloat(Value=&FloatFromCurve(FloatCurve=DataInterface() {propertyValues {Curve {keys [<at 0 0>,<at 0.3 1>,<at 0.6 1.075>,<at 1 0>]}}})))
   &CameraOffset(Camera Offset Amount=3) {enabled=false}
   &DynamicMaterialParameters(
    Index 0 Param 1=&FloatFromCurve(
     FloatCurve=DataInterface() {propertyValues {Curve {keys [{tangentMode=RCTM_User, time=0, value=0}, {tangentMode=RCTM_User, time=1, value=1}]}}}
     CurveIndex=$Engine.System.Age
    )
    Module.Index0Float0Write=true
    Module.Index0Float1Write=true
    Module.Index0Float2Write=true
    Module.Index0Float3Write=true
   )
   &CameraOffset(Camera Offset Amount=5) {module=CameraOffset001}
  ]
  rendererClasses [&NiagaraSpriteRendererProperties]
  renderers [
   {rendererClass=&NiagaraSpriteRendererProperties}
   {rendererClass=&NiagaraSpriteRendererProperties, rendererIndex=1}
  ]
 }
}
summary {
 systemName=NS_Emote_Guitar_RainbowSerenade_Rainbow
 emitters [
  {emitterName=Rainbow, rendererClasses [&NiagaraMeshRendererProperties]}
  {emitterName=SeatCloudsTraversal, rendererClasses [&NiagaraMeshRendererProperties]}
  {emitterName=SparklesRainbow, rendererClasses [&NiagaraSpriteRendererProperties]}
  {emitterName=SparklesTraversal, rendererClasses [&NiagaraSpriteRendererProperties]}
  {emitterName=SideClouds, rendererClasses [&NiagaraMeshRendererProperties]}
  {emitterName=SideCloudsTraversal, rendererClasses [&NiagaraMeshRendererProperties]}
 ]
 userVariables [
  User.bIsFrontEnd {type=NiagaraBool}
  User.bIsFrontEndPreview {type=NiagaraBool}
 ]
}
